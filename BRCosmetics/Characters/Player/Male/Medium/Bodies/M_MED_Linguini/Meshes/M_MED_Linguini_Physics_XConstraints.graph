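# Physics asset for the M_MED_Linguini male player body, modeling the dynamic lower jacket with capsules on the pelvis, legs and feet.
# Six jacket chains (front, side and back, left and right) run from pelvis-attached master bodies through five segments with limited swing and tight twist.
# Angular-free cross constraints link the fourth segments front to side to back and across the back, tying the chains into a band open at the front.
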PhysicsAsset /BRCosmetics/Characters/Player/Male/Medium/Bodies/M_MED_Linguini/Meshes/M_MED_Linguini_Physics_XConstraints
bodies {
 calf_l [{shapeType=Capsule, length=35, radius=11, center=<-24 0 0>, rotation=<rot 90 -8 -14>}]
 calf_r [{shapeType=Capsule, length=35, radius=11, center=<24 0 0>, rotation=<rot 90 -12 -17>}]
 dyn_jacketBack_01_l [{shapeType=Capsule, length=18, radius=10, center=<4 0 6>, rotation=<rot 0 0 90>}]
 dyn_jacketBack_01_r [{shapeType=Capsule, length=18, radius=10, center=<4 0 6>, rotation=<rot 0 0 90>}]
 dyn_jacketBack_02_l [{shapeType=Capsule, length=19, radius=10, center=<4 0 6>, rotation=<rot 0 0 90>}]
 dyn_jacketBack_02_r [{shapeType=Capsule, length=19, radius=10, center=<4 0 6>, rotation=<rot 0 0 90>}]
 dyn_jacketBack_03_l [{shapeType=Capsule, length=20, radius=10, center=<6.8942 0 5.9997>, rotation=<rot 0 0 90>}]
 dyn_jacketBack_03_r [{shapeType=Capsule, length=20, radius=10, center=<6.8942 0 5.9997>, rotation=<rot 0 0 90>}]
 dyn_jacketBack_04_l [{shapeType=Capsule, length=21, radius=10, center=<9.9453 0.0001 0.3355>, rotation=<rot 0 0 90>}]
 dyn_jacketBack_04_r [{shapeType=Capsule, length=21, radius=10, center=<9.9453 0.0002 -0.3498>, rotation=<rot 0 0 89.9999>}]
 dyn_jacketBack_05_l [{shapeType=Capsule, length=22, radius=9, center=<9.9453 0 -1.2516>, rotation=<rot 0 0 90>}]
 dyn_jacketBack_05_r [{shapeType=Capsule, length=22, radius=9, center=<9.9453 0 -1.2516>, rotation=<rot 0 0 90>}]
 dyn_jacketBack_mstr_l [{shapeType=Capsule, length=0.505, radius=0.505, rotation=<rot 0 -90 0>}]
 dyn_jacketBack_mstr_r [{shapeType=Capsule, length=0.505, radius=0.505, rotation=<rot 0 -90 0>}]
 dyn_jacketFront_01_l [{shapeType=Capsule, length=7, radius=10, center=<4 0 9>, rotation=<rot 90 0 0>}]
 dyn_jacketFront_01_r [{shapeType=Capsule, length=7, radius=10, center=<4 0 9>, rotation=<rot 90 0 0>}]
 dyn_jacketFront_02_l [{shapeType=Capsule, length=7, radius=10, center=<4 0 9>, rotation=<rot 90 0 0>}]
 dyn_jacketFront_02_r [{shapeType=Capsule, length=7, radius=10, center=<4 0 9>, rotation=<rot 90 0 0>}]
 dyn_jacketFront_03_l [{shapeType=Capsule, length=7, radius=10, center=<4 0 9>, rotation=<rot 90 0 0>}]
 dyn_jacketFront_03_r [{shapeType=Capsule, length=7, radius=10, center=<4 0 9>, rotation=<rot 90 0 0>}]
 dyn_jacketFront_04_l [{shapeType=Capsule, length=7, radius=10, center=<4 0 9>, rotation=<rot 90 0 0>}]
 dyn_jacketFront_04_r [{shapeType=Capsule, length=7, radius=10, center=<4 0 9>, rotation=<rot 90 0 0>}]
 dyn_jacketFront_05_l [{shapeType=Capsule, length=7, radius=10, center=<4 0 9>, rotation=<rot 90 0 0>}]
 dyn_jacketFront_05_r [{shapeType=Capsule, length=7, radius=10, center=<4 0 9>, rotation=<rot 90 0 0>}]
 dyn_jacketFront_mstr_l [{shapeType=Capsule, length=0.505, radius=0.505, rotation=<rot 0 -90 0>}]
 dyn_jacketFront_mstr_r [{shapeType=Capsule, length=0.505, radius=0.505, rotation=<rot 0 -90 0>}]
 dyn_jacketSide_01_l [{shapeType=Capsule, length=18, radius=10, center=<4 -2 9>, rotation=<rot 0 0 90>}]
 dyn_jacketSide_01_r [{shapeType=Capsule, length=18, radius=10, center=<4 2 9>, rotation=<rot 0 0 90>}]
 dyn_jacketSide_02_l [{shapeType=Capsule, length=19, radius=10, center=<4 -2 9>, rotation=<rot 0 0 90>}]
 dyn_jacketSide_02_r [{shapeType=Capsule, length=19, radius=10, center=<4 2 9>, rotation=<rot 0 0 90>}]
 dyn_jacketSide_03_l [{shapeType=Capsule, length=20, radius=10, center=<4 -2 9>, rotation=<rot 0 0 90>}]
 dyn_jacketSide_03_r [{shapeType=Capsule, length=20, radius=10, center=<4 2 9>, rotation=<rot 0 0 90>}]
 dyn_jacketSide_04_l [{shapeType=Capsule, length=21, radius=10, center=<4 -2 9>, rotation=<rot 0 0 90>}]
 dyn_jacketSide_04_r [{shapeType=Capsule, length=21, radius=10, center=<7.0562 2 3.7295>, rotation=<rot 0 0 89.9999>}]
 dyn_jacketSide_05_l [{shapeType=Capsule, length=22, radius=10, center=<4 -2 9>, rotation=<rot 0 0 90>}]
 dyn_jacketSide_05_r [{shapeType=Capsule, length=22, radius=10, center=<9.9843 1.9999 4.3829>, rotation=<rot 0 0 89.9999>}]
 dyn_jacketSide_mstr_l [{shapeType=Capsule, length=0.505, radius=0.505, rotation=<rot 0 -90 0>}]
 dyn_jacketSide_mstr_r [{shapeType=Capsule, length=0.505, radius=0.505, rotation=<rot 0 -90 0>}]
 foot_l [{shapeType=Capsule, length=12.5616, radius=11.8777, shapeName=foot_l_capsule, center=<0 6.2288 0.1369>, rotation=<rot 0 0 88.7409>}]
 foot_r [{shapeType=Capsule, length=12.5616, radius=11.8777, shapeName=foot_r_capsule, center=<2.6584 -4.043 -0.0891>, rotation=<rot 0 0 88.7409>}]
 pelvis [{shapeType=Capsule, length=10, radius=10, center=<-2 0 0>}]
 thigh_l [{shapeType=Capsule, length=30, radius=9, center=<-13 0 0>, rotation=<rot 90 0 0>}]
 thigh_r [{shapeType=Capsule, length=30, radius=9, center=<13 0 0>, rotation=<rot 90 0 180>}]
}
constraints [
 {
  bone1Name=thigh_l
  bone2Name=pelvis
  swing1LimitDegrees=45
  swing1Motion=Free
  swing2LimitDegrees=45
  swing2Motion=Free
  twistLimitDegrees=45
  twistMotion=Free
 }
 {
  bone1Name=calf_l
  bone2Name=thigh_l
  swing1LimitDegrees=45
  swing1Motion=Free
  swing2LimitDegrees=45
  swing2Motion=Free
  twistLimitDegrees=45
  twistMotion=Free
 }
 {
  bone1Name=thigh_r
  bone2Name=pelvis
  swing1LimitDegrees=45
  swing1Motion=Free
  swing2LimitDegrees=45
  swing2Motion=Free
  twistLimitDegrees=45
  twistMotion=Free
 }
 {
  bone1Name=calf_r
  bone2Name=thigh_r
  swing1LimitDegrees=45
  swing1Motion=Free
  swing2LimitDegrees=45
  swing2Motion=Free
  twistLimitDegrees=45
  twistMotion=Free
 }
 {bone1Name=dyn_jacketSide_01_l, bone2Name=dyn_jacketSide_mstr_l, swing1LimitDegrees=45, swing2LimitDegrees=45, twistLimitDegrees=5}
 {bone1Name=dyn_jacketSide_02_l, bone2Name=dyn_jacketSide_01_l, swing1LimitDegrees=45, swing2LimitDegrees=45, twistLimitDegrees=5}
 {bone1Name=dyn_jacketSide_03_l, bone2Name=dyn_jacketSide_02_l, swing1LimitDegrees=45, swing2LimitDegrees=45, twistLimitDegrees=5}
 {bone1Name=dyn_jacketSide_04_l, bone2Name=dyn_jacketSide_03_l, swing1LimitDegrees=45, swing2LimitDegrees=45, twistLimitDegrees=5}
 {bone1Name=dyn_jacketSide_05_l, bone2Name=dyn_jacketSide_04_l, swing1LimitDegrees=45, swing2LimitDegrees=45, twistLimitDegrees=5}
 {bone1Name=dyn_jacketBack_01_l, bone2Name=dyn_jacketBack_mstr_l, swing1LimitDegrees=45, swing2LimitDegrees=45, twistLimitDegrees=5}
 {bone1Name=dyn_jacketBack_02_l, bone2Name=dyn_jacketBack_01_l, swing1LimitDegrees=45, swing2LimitDegrees=45, twistLimitDegrees=5}
 {bone1Name=dyn_jacketBack_03_l, bone2Name=dyn_jacketBack_02_l, swing1LimitDegrees=45, swing2LimitDegrees=45, twistLimitDegrees=5}
 {bone1Name=dyn_jacketBack_04_l, bone2Name=dyn_jacketBack_03_l, swing1LimitDegrees=45, swing2LimitDegrees=45, twistLimitDegrees=5}
 {bone1Name=dyn_jacketBack_05_l, bone2Name=dyn_jacketBack_04_l, swing1LimitDegrees=45, swing2LimitDegrees=45, twistLimitDegrees=5}
 {bone1Name=dyn_jacketBack_01_r, bone2Name=dyn_jacketBack_mstr_r, swing1LimitDegrees=45, swing2LimitDegrees=45, twistLimitDegrees=5}
 {bone1Name=dyn_jacketBack_02_r, bone2Name=dyn_jacketBack_01_r, swing1LimitDegrees=45, swing2LimitDegrees=45, twistLimitDegrees=5}
 {bone1Name=dyn_jacketBack_03_r, bone2Name=dyn_jacketBack_02_r, swing1LimitDegrees=45, swing2LimitDegrees=45, twistLimitDegrees=5}
 {bone1Name=dyn_jacketBack_04_r, bone2Name=dyn_jacketBack_03_r, swing1LimitDegrees=45, swing2LimitDegrees=45, twistLimitDegrees=5}
 {bone1Name=dyn_jacketBack_05_r, bone2Name=dyn_jacketBack_04_r, swing1LimitDegrees=45, swing2LimitDegrees=45, twistLimitDegrees=5}
 {bone1Name=dyn_jacketSide_01_r, bone2Name=dyn_jacketSide_mstr_r, swing1LimitDegrees=45, swing2LimitDegrees=45, twistLimitDegrees=5}
 {bone1Name=dyn_jacketSide_02_r, bone2Name=dyn_jacketSide_01_r, swing1LimitDegrees=45, swing2LimitDegrees=45, twistLimitDegrees=5}
 {bone1Name=dyn_jacketSide_03_r, bone2Name=dyn_jacketSide_02_r, swing1LimitDegrees=45, swing2LimitDegrees=45, twistLimitDegrees=5}
 {bone1Name=dyn_jacketSide_04_r, bone2Name=dyn_jacketSide_03_r, swing1LimitDegrees=45, swing2LimitDegrees=45, twistLimitDegrees=5}
 {bone1Name=dyn_jacketSide_05_r, bone2Name=dyn_jacketSide_04_r, swing1LimitDegrees=45, swing2LimitDegrees=45, twistLimitDegrees=5}
 {
  bone1Name=dyn_jacketSide_mstr_l
  bone2Name=pelvis
  swing1LimitDegrees=45
  swing1Motion=Free
  swing2LimitDegrees=45
  swing2Motion=Free
  twistLimitDegrees=45
  twistMotion=Free
 }
 {
  bone1Name=dyn_jacketBack_mstr_l
  bone2Name=pelvis
  swing1LimitDegrees=45
  swing1Motion=Free
  swing2LimitDegrees=45
  swing2Motion=Free
  twistLimitDegrees=45
  twistMotion=Free
 }
 {
  bone1Name=dyn_jacketBack_mstr_r
  bone2Name=pelvis
  swing1LimitDegrees=45
  swing1Motion=Free
  swing2LimitDegrees=45
  swing2Motion=Free
  twistLimitDegrees=45
  twistMotion=Free
 }
 {
  bone1Name=dyn_jacketSide_mstr_r
  bone2Name=pelvis
  swing1LimitDegrees=45
  swing1Motion=Free
  swing2LimitDegrees=45
  swing2Motion=Free
  twistLimitDegrees=45
  twistMotion=Free
 }
 {
  bone1Name=dyn_jacketFront_mstr_l
  bone2Name=pelvis
  swing1LimitDegrees=45
  swing1Motion=Free
  swing2LimitDegrees=45
  swing2Motion=Free
  twistLimitDegrees=45
  twistMotion=Free
 }
 {bone1Name=dyn_jacketFront_01_l, bone2Name=dyn_jacketFront_mstr_l, swing1LimitDegrees=45, swing2LimitDegrees=45, twistLimitDegrees=5}
 {bone1Name=dyn_jacketFront_02_l, bone2Name=dyn_jacketFront_01_l, swing1LimitDegrees=45, swing2LimitDegrees=45, twistLimitDegrees=5}
 {bone1Name=dyn_jacketFront_03_l, bone2Name=dyn_jacketFront_02_l, swing1LimitDegrees=45, swing2LimitDegrees=45, twistLimitDegrees=5}
 {bone1Name=dyn_jacketFront_04_l, bone2Name=dyn_jacketFront_03_l, swing1LimitDegrees=45, swing2LimitDegrees=45, twistLimitDegrees=5}
 {bone1Name=dyn_jacketFront_05_l, bone2Name=dyn_jacketFront_04_l, swing1LimitDegrees=45, swing2LimitDegrees=45, twistLimitDegrees=5}
 {
  bone1Name=dyn_jacketFront_mstr_r
  bone2Name=pelvis
  swing1LimitDegrees=45
  swing1Motion=Free
  swing2LimitDegrees=45
  swing2Motion=Free
  twistLimitDegrees=45
  twistMotion=Free
 }
 {bone1Name=dyn_jacketFront_01_r, bone2Name=dyn_jacketFront_mstr_r, swing1LimitDegrees=45, swing2LimitDegrees=45, twistLimitDegrees=5}
 {bone1Name=dyn_jacketFront_02_r, bone2Name=dyn_jacketFront_01_r, swing1LimitDegrees=45, swing2LimitDegrees=45, twistLimitDegrees=5}
 {bone1Name=dyn_jacketFront_03_r, bone2Name=dyn_jacketFront_02_r, swing1LimitDegrees=45, swing2LimitDegrees=45, twistLimitDegrees=5}
 {bone1Name=dyn_jacketFront_04_r, bone2Name=dyn_jacketFront_03_r, swing1LimitDegrees=45, swing2LimitDegrees=45, twistLimitDegrees=5}
 {bone1Name=dyn_jacketFront_05_r, bone2Name=dyn_jacketFront_04_r, swing1LimitDegrees=45, swing2LimitDegrees=45, twistLimitDegrees=5}
 {
  bone1Name=dyn_jacketSide_04_r
  bone2Name=dyn_jacketBack_04_r
  swing1LimitDegrees=45
  swing1Motion=Free
  swing2LimitDegrees=45
  swing2Motion=Free
  twistLimitDegrees=45
  twistMotion=Free
 }
 {
  bone1Name=dyn_jacketSide_04_l
  bone2Name=dyn_jacketBack_04_l
  swing1LimitDegrees=45
  swing1Motion=Free
  swing2LimitDegrees=45
  swing2Motion=Free
  twistLimitDegrees=45
  twistMotion=Free
 }
 {
  bone1Name=dyn_jacketBack_04_r
  bone2Name=dyn_jacketBack_04_l
  swing1LimitDegrees=45
  swing1Motion=Free
  swing2LimitDegrees=45
  swing2Motion=Free
  twistLimitDegrees=45
  twistMotion=Free
 }
 {
  bone1Name=dyn_jacketFront_04_r
  bone2Name=dyn_jacketSide_04_r
  swing1LimitDegrees=45
  swing1Motion=Free
  swing2LimitDegrees=45
  swing2Motion=Free
  twistLimitDegrees=45
  twistMotion=Free
 }
 {
  bone1Name=dyn_jacketFront_04_l
  bone2Name=dyn_jacketSide_04_l
  swing1LimitDegrees=45
  swing1Motion=Free
  swing2LimitDegrees=45
  swing2Motion=Free
  twistLimitDegrees=45
  twistMotion=Free
 }
]
