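# Post-copy AnimBP for the Linguini male body cosmetic, adding secondary jacket motion to the pose copied from the attached parent mesh.
# Additive bone tweaks adjust the back jacket bones, plus left side and front while crouching, then cache as preSim.
# Two near-identical RigidBody sims read that cache with a gravity override from a deprecated state function, blended by bIsSkyDiving.
# The event graph also sets run_alpha from pawn speed, unused in the AnimGraph.
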AnimBlueprint /BRCosmetics/Characters/Player/Male/Medium/Bodies/M_MED_Linguini/Meshes/M_MED_Linguini_AnimBP
target_skeleton=&/M_MED_Linguini_Skeleton
graphs [
 AnimGraph {nodes [
   n1 = Get GravityOverride()
   n2 = RigidBody!(
    ComponentPose=Local To Component(LocalPose=Use cached pose 'preSim'())
    bAlphaBoolEnabled:bool
    OverrideWorldGravity:Vector=@n1
    ExternalForce
   ) -> (Pose) {
    alphaInputType=Bool
    bClampLinearTranslationLimitToRefPose=true
    bOverrideWorldGravity=true
    overridePhysicsAsset=&/M_MED_Linguini_Physics_XConstraints
    simulationSpace=BaseBoneSpace
    alphaBoolBlend {blendOption=Linear}
    baseBoneRef {boneName=root}
    componentAppliedLinearAccClamp=<3000 3000 3000>
    componentLinearAccScale=<0.5 0.4 0.4>
    componentLinearVelScale=<0.5 0.4 0.4>
    overrideWorldGravity {z=-980}
   }
   n3 = RigidBody!(ComponentPose=Local To Component(LocalPose=Use cached pose 'preSim'()), Alpha:float, OverrideWorldGravity=@n1, ExternalForce) -> (Pose) {
    bClampLinearTranslationLimitToRefPose=true
    bOverrideWorldGravity=true
    overridePhysicsAsset=&M_MED_Linguini_Physics_XConstraints
    simulationSpace=BaseBoneSpace
    alphaBoolBlend {blendOption=Linear}
    baseBoneRef {boneName=root}
    componentAppliedLinearAccClamp=<3000 3000 3000>
    componentLinearAccScale=<0.5 0.4 0.4>
    componentLinearVelScale=<0.5 0.4 0.4>
    overrideWorldGravity {z=-980}
   }
   Output Pose(Result=Blend Poses by bool(BlendPose_0=Component To Local(ComponentPose=@n2), BlendPose_1=Component To Local(ComponentPose=@n3), bActiveValue=Get bIsSkyDiving()) {blendType=Linear}) {layerGroup=None}
   n4 = "Transform (Modify) Bone - Bone: dyn_jacketBack_mstr_r"(
    ComponentPose=Local To Component(LocalPose=Copy Pose From Mesh() {bUseAttachedParent=true})
    Alpha=1
    Translation="-0.746,-1.508,0.717"
    Rotation="13.459,6.003,1.464"
    Scale="1,1,1"
   ) {
    rotationMode=BMM_Additive
    rotationSpace=BCS_BoneSpace
    translationMode=BMM_Additive
    translationSpace=BCS_BoneSpace
    alphaBoolBlend {blendOption=Linear}
    boneToModify {boneName=dyn_jacketBack_mstr_r}
    rotation=<rot 13.459 6.0031 1.464>
    translation=<-0.746 -1.508 0.717>
   }
   n5 = "Transform (Modify) Bone - Bone: dyn_jacketBack_mstr_l"(ComponentPose=@n4, Alpha=1, Translation="0.7,-1.5,0.5", Rotation="13.808,-4.073,-2.255", Scale="1,1,1") {
    rotationMode=BMM_Additive
    rotationSpace=BCS_BoneSpace
    translationMode=BMM_Additive
    translationSpace=BCS_BoneSpace
    alphaBoolBlend {blendOption=Linear}
    boneToModify {boneName=dyn_jacketBack_mstr_l}
    rotation=<rot 13.808 -4.073 -2.255>
    translation=<0.7 -1.5 0.5>
   }
   n6 = Get bIsCrouching()
   n7 = "Transform (Modify) Bone - Bone: dyn_jacketSide_mstr_l"!(ComponentPose=@n5, bAlphaBoolEnabled:bool=@n6, Translation="3,-2,2", Rotation="0,0,0", Scale="1,1,1") -> (Pose) {
    alphaInputType=Bool
    rotationMode=BMM_Additive
    translationMode=BMM_Additive
    alphaBoolBlend {blendInTime=1, blendOption=Linear, blendOutTime=1}
    boneToModify {boneName=dyn_jacketSide_mstr_l}
    translation=<3 -2 2>
   }
   n8 = "Transform (Modify) Bone - Bone: dyn_jacketFront_mstr_l"!(ComponentPose=@n7, bAlphaBoolEnabled:bool=@n6, Translation="3,-2,2", Rotation="5,0,0", Scale="1,1,1") -> (Pose) {
    alphaInputType=Bool
    rotationMode=BMM_Additive
    translationMode=BMM_Additive
    alphaBoolBlend {blendInTime=1, blendOption=Linear, blendOutTime=1}
    boneToModify {boneName=dyn_jacketFront_mstr_l}
    rotation {pitch=5}
    translation=<3 -2 2>
   }
   Save cached pose 'preSim'(Pose=Component To Local(ComponentPose=@n8)) {cacheName=preSim}
  ]}
 EventGraph {nodes [
   Event BlueprintUpdateAnimation()
   > n1 = Sequence()
   > n2 = Gravity Override State Function DEPRECATED(CharacterPartAnimInstance=Self-Reference(), GameTimeInSeconds=GetGameTimeInSeconds(), GravityOverrideParams=Get Gravity Override Params())
   > Set GravityOverride(GravityOverride=@n2)
   n3 = Branch(execute=@n1.then_1, Condition="float > float"(A=Get PawnSpeed2D(), B=50) {is=title K2Node_CallFunction})
   > Set run_alpha(run_alpha:double=100) -> (Output_Get:double)
   Set run_alpha(execute=@n3.else, run_alpha=0)
  ]}
]
variables [
 Gravity Override Params {
  type=GravityOverrideParamsStruct
  default {
   skydive_DivingDown {gravityOverride=<0 0 0>}
   skydive_FloatingDown {gravityOverride=<0 0 0>}
   skydive_GoingUp {gravityOverride=<0 -300 -100>}
  }
 }
 GravityOverride {type=Vector}
 run_alpha {type=number}
]
