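# Physics asset for a part of the Linguini medium male player cosmetic, driving left and right dynamic hair bone chains.
# Each chain links four identical capsules in series from the head, with a very narrow second swing limit so joints bend mostly in one plane.
# A small head capsule links to a large upper-spine capsule with free rotation, and another rotation-free constraint ties the two chains at their second segments.
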PhysicsAsset /BRCosmetics/Characters/Player/Male/Medium/Bodies/M_MED_Linguini/Meshes/Parts/M_MED_Linguini_PhysicsPack
bodies {
 dyn_hair_01_l [{shapeType=Capsule, length=3, radius=4, rotation=<rot 0 0 90>}]
 dyn_hair_01_r [{shapeType=Capsule, length=3, radius=4, rotation=<rot 0 0 90>}]
 dyn_hair_02_l [{shapeType=Capsule, length=3, radius=4, rotation=<rot 0 0 90>}]
 dyn_hair_02_r [{shapeType=Capsule, length=3, radius=4, rotation=<rot 0 0 90>}]
 dyn_hair_03_l [{shapeType=Capsule, length=3, radius=4, rotation=<rot 0 0 90>}]
 dyn_hair_03_r [{shapeType=Capsule, length=3, radius=4, rotation=<rot 0 0 90>}]
 dyn_hair_04_l [{shapeType=Capsule, length=3, radius=4, rotation=<rot 0 0 90>}]
 dyn_hair_04_r [{shapeType=Capsule, length=3, radius=4, rotation=<rot 0 0 90>}]
 head [{shapeType=Capsule, length=1, radius=1}]
 spine_05 [{shapeType=Capsule, length=25, radius=12.5, center=<3.8467 3.5627 0>}]
}
constraints [
 {bone1Name=dyn_hair_01_l, bone2Name=head, swing1LimitDegrees=45, swing2LimitDegrees=1, twistLimitDegrees=15}
 {bone1Name=dyn_hair_02_l, bone2Name=dyn_hair_01_l, swing1LimitDegrees=45, swing2LimitDegrees=1, twistLimitDegrees=15}
 {bone1Name=dyn_hair_03_l, bone2Name=dyn_hair_02_l, swing1LimitDegrees=45, swing2LimitDegrees=1, twistLimitDegrees=15}
 {bone1Name=dyn_hair_04_l, bone2Name=dyn_hair_03_l, swing1LimitDegrees=45, swing2LimitDegrees=1, twistLimitDegrees=15}
 {bone1Name=dyn_hair_01_r, bone2Name=head, swing1LimitDegrees=45, swing2LimitDegrees=1, twistLimitDegrees=15}
 {bone1Name=dyn_hair_02_r, bone2Name=dyn_hair_01_r, swing1LimitDegrees=45, swing2LimitDegrees=1, twistLimitDegrees=15}
 {bone1Name=dyn_hair_03_r, bone2Name=dyn_hair_02_r, swing1LimitDegrees=45, swing2LimitDegrees=1, twistLimitDegrees=15}
 {bone1Name=dyn_hair_04_r, bone2Name=dyn_hair_03_r, swing1LimitDegrees=45, swing2LimitDegrees=1, twistLimitDegrees=15}
 {
  bone1Name=head
  bone2Name=spine_05
  swing1LimitDegrees=45
  swing1Motion=Free
  swing2LimitDegrees=45
  swing2Motion=Free
  twistLimitDegrees=45
  twistMotion=Free
 }
 {
  bone1Name=dyn_hair_02_r
  bone2Name=dyn_hair_02_l
  swing1LimitDegrees=45
  swing1Motion=Free
  swing2LimitDegrees=45
  swing2Motion=Free
  twistLimitDegrees=45
  twistMotion=Free
 }
]
